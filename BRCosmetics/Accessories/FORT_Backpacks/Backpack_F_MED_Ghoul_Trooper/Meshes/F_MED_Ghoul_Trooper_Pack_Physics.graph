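# Physics asset for the Ghoul Trooper backpack cosmetic, defining bodies and joints for its dangling ears, eye, fur, tear and brain pieces.
# Box bodies cover the pack, with pack and pack_mid constrained to root.
# Each dyn_ piece hangs from a small _mstr anchor body on pack_mid, with swing and twist limits tuned per piece.
# The brain anchor is the exception, with free swing and twist against pack_mid.
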PhysicsAsset /BRCosmetics/Accessories/FORT_Backpacks/Backpack_F_MED_Ghoul_Trooper/Meshes/F_MED_Ghoul_Trooper_Pack_Physics
bodies {
 dyn_brain [{shapeType=Box, extentX=2, extentY=2, extentZ=2, center=<-0.6 0 0>}]
 dyn_brain_mstr [{shapeType=Sphere, radius=0.505}]
 dyn_ear_l [{shapeType=Box, extentX=6, extentY=2, extentZ=4, center=<-2.024 -0.0001 0>, rotation=<rot 90 0 0>}]
 dyn_ear_l_mstr [{shapeType=Box, extentX=1.01, extentY=1.01, extentZ=1.01}]
 dyn_ear_r [{shapeType=Box, extentX=6, extentY=2, extentZ=4, center=<2 0 0>, rotation=<rot 90 0 0>}]
 dyn_ear_r_mstr [{shapeType=Box, extentX=1.01, extentY=1.01, extentZ=1.01}]
 dyn_eye [{shapeType=Box, extentX=3.672, extentY=4.5649, extentZ=4, center=<2 0 0>, rotation=<rot 90 0 0>}]
 dyn_eye_mstr [{shapeType=Box, extentX=1.01, extentY=1.01, extentZ=1.01}]
 dyn_fur_l [{shapeType=Box, extentX=6, extentY=2, extentZ=4, center=<2 0 0>, rotation=<rot 90 0 0>}]
 dyn_fur_l_mstr [{shapeType=Box, extentX=1.01, extentY=1.01, extentZ=1.01}]
 dyn_fur_r [{shapeType=Box, extentX=6, extentY=2, extentZ=4, center=<2 0 0>, rotation=<rot 90 0 0>}]
 dyn_fur_r_mstr [{shapeType=Box, extentX=1.01, extentY=1.01, extentZ=1.01}]
 dyn_tear [{shapeType=Box, extentX=0.5394, extentY=4.3907, extentZ=2.5394, center=<1.25 0 0>, rotation=<rot 90 0 0>}]
 dyn_tear_mstr [{shapeType=Box, extentX=1.01, extentY=1.01, extentZ=1.01}]
 pack [{shapeType=Box, extentX=19.3382, extentY=10.866, extentZ=11.3001, center=<-0.0459 -9.2168 0>}]
 pack_mid [{shapeType=Box, extentX=24.6928, extentY=6.8829, extentZ=14.07, center=<0 -3.1345 2.2012>}]
 root [{shapeType=Box, extentX=1.01, extentY=1.01, extentZ=1.01}]
}
constraints [
 {bone1Name=pack, bone2Name=root, swing1LimitDegrees=25, swing2LimitDegrees=25, twistLimitDegrees=5}
 {bone1Name=pack_mid, bone2Name=root, swing1LimitDegrees=25, swing2LimitDegrees=25, twistLimitDegrees=5}
 {bone1Name=dyn_ear_r_mstr, bone2Name=pack_mid, swing1LimitDegrees=25, swing2LimitDegrees=25, twistLimitDegrees=5}
 {bone1Name=dyn_ear_l_mstr, bone2Name=pack_mid, swing1LimitDegrees=25, swing2LimitDegrees=25, twistLimitDegrees=5}
 {bone1Name=dyn_fur_l_mstr, bone2Name=pack_mid, swing1LimitDegrees=25, swing2LimitDegrees=25, twistLimitDegrees=5}
 {bone1Name=dyn_fur_r_mstr, bone2Name=pack_mid, swing1LimitDegrees=25, swing2LimitDegrees=25, twistLimitDegrees=5}
 {bone1Name=dyn_tear_mstr, bone2Name=pack_mid, swing1LimitDegrees=25, swing2LimitDegrees=25, twistLimitDegrees=5}
 {bone1Name=dyn_eye_mstr, bone2Name=pack_mid, swing1LimitDegrees=25, swing2LimitDegrees=25, twistLimitDegrees=5}
 {bone1Name=dyn_ear_r, bone2Name=dyn_ear_r_mstr, swing1LimitDegrees=25, swing2LimitDegrees=12.5, twistLimitDegrees=5}
 {bone1Name=dyn_ear_l, bone2Name=dyn_ear_l_mstr, swing1LimitDegrees=25, swing2LimitDegrees=12.5, twistLimitDegrees=5}
 {bone1Name=dyn_fur_l, bone2Name=dyn_fur_l_mstr, swing1LimitDegrees=40, swing2LimitDegrees=45, twistLimitDegrees=5}
 {bone1Name=dyn_fur_r, bone2Name=dyn_fur_r_mstr, swing1LimitDegrees=40, swing2LimitDegrees=45, twistLimitDegrees=5}
 {bone1Name=dyn_tear, bone2Name=dyn_tear_mstr, swing1LimitDegrees=42.1429, swing2LimitDegrees=81.5714, twistLimitDegrees=5}
 {bone1Name=dyn_eye, bone2Name=dyn_eye_mstr, swing1LimitDegrees=45, swing2LimitDegrees=45, twistLimitDegrees=30}
 {bone1Name=dyn_brain, bone2Name=dyn_brain_mstr, swing1LimitDegrees=60, swing2LimitDegrees=90, twistLimitDegrees=5}
 {
  bone1Name=dyn_brain_mstr
  bone2Name=pack_mid
  swing1LimitDegrees=45
  swing1Motion=Free
  swing2LimitDegrees=45
  swing2Motion=Free
  twistLimitDegrees=45
  twistMotion=Free
 }
]
